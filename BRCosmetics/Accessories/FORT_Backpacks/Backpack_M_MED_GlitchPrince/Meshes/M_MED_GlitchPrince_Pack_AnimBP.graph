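# Animation Blueprint for the Glitch Prince backpack cosmetic, adjusting the pack to spine bend and adding secondary motion.
# Each update maps the yaw of spine_05's ref-pose delta relative to spine_01 into a roll and Y offset, applied additively to the pack bone.
# Chained Anim Dynamics nodes then simulate wing_l, wing_r and wire with angular springs, limits, overridden gravity and scaled component motion.
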AnimBlueprint /BRCosmetics/Accessories/FORT_Backpacks/Backpack_M_MED_GlitchPrince/Meshes/M_MED_GlitchPrince_Pack_AnimBP
target_skeleton=&/M_MED_GlitchPrince_Pack_Skeleton
graphs [
 AnimGraph {nodes [
   n1 = "Transform (Modify) Bone - Bone: pack"(ComponentPose=Local To Component(LocalPose=Input Pose()), Translation=Get backpack_translation(), Rotation=Get backpack_rotate()) {rotationMode=BMM_Additive, translationMode=BMM_Additive, alphaBoolBlend {blendOption=Linear}, boneToModify {boneName=pack}}
   n2 = Anim Dynamics - Bone: wing_l!(SimulationSpace:AnimPhysSimSpaceType Enum, ComponentPose=@n1, Alpha=1, GravityOverride:Vector="0,0,-980", ExternalForce="0,0,0") -> (Pose) {
    angularDampingOverride=1
    angularSpringConstant=1200
    bAngularSpring=true
    bGravityOverrideInSimSpace=true
    bOverrideAngularDamping=true
    bUseGravityOverride=true
    alphaBoolBlend {blendOption=Linear}
    boundBone {boneName=wing_l}
    componentAppliedLinearAccClamp=<3000 3000 2500>
    componentLinearAccScale=<0.5 0.5 0.5>
    componentLinearVelScale=<0.5 0.5 0.5>
    physicsBodyDefinitions [{
      boundBone {boneName=wing_l}
      boxExtents=<10 10 10>
      constraintSetup {coneAngle=1, angularLimitsMax=<0 0 15>, angularLimitsMin=<0 0 -25>, angularTarget=<1 0 0>}
      localJointOffset=<0.5 0 0>
     }]
   }
   n3 = Anim Dynamics - Bone: wing_r!(SimulationSpace:AnimPhysSimSpaceType Enum, ComponentPose=@n2, Alpha=1, GravityOverride:Vector="0,0,-980", ExternalForce="0,0,0") -> (Pose) {
    angularDampingOverride=1
    angularSpringConstant=1200
    bAngularSpring=true
    bGravityOverrideInSimSpace=true
    bOverrideAngularDamping=true
    bUseGravityOverride=true
    alphaBoolBlend {blendOption=Linear}
    boundBone {boneName=wing_r}
    componentAppliedLinearAccClamp=<3000 3000 2500>
    componentLinearAccScale=<0.5 0.5 0.5>
    componentLinearVelScale=<0.5 0.5 0.5>
    physicsBodyDefinitions [{
      boundBone {boneName=wing_r}
      boxExtents=<10 10 10>
      constraintSetup {coneAngle=1, angularLimitsMax=<0 0 15>, angularLimitsMin=<0 0 -25>, angularTarget=<1 0 0>}
      localJointOffset=<0.5 0 0>
     }]
   }
   n4 = Anim Dynamics - Bone: wire!(SimulationSpace:AnimPhysSimSpaceType Enum, ComponentPose=@n3, Alpha=1, GravityOverride:Vector="0,0,-980", ExternalForce="0,0,0") -> (Pose) {
    angularDampingOverride=1
    angularSpringConstant=1000
    bAngularSpring=true
    bGravityOverrideInSimSpace=true
    bOverrideAngularDamping=true
    bOverrideLinearDamping=true
    bUseGravityOverride=true
    linearDampingOverride=1
    linearSpringConstant=100
    alphaBoolBlend {blendOption=Linear}
    boundBone {boneName=wire}
    componentAppliedLinearAccClamp=<3000 3000 2500>
    componentLinearAccScale=<0.35 0.35 0.35>
    componentLinearVelScale=<0.35 0.35 0.35>
    physicsBodyDefinitions [{
      boundBone {boneName=wire}
      boxExtents=<8 8 8>
      constraintSetup {coneAngle=1, angularLimitsMax=<0 0 20>, angularLimitsMin=<0 0 -45>, angularTarget=<1 0 0>}
      localJointOffset=<2 0 0>
     }]
   }
   Output Pose(Result=Component To Local(ComponentPose=@n4)) {layerGroup=None}
  ]}
 EventGraph {nodes [
   Event BlueprintUpdateAnimation()
   n1 = GetOwnerSkeletalMeshComponent()
   n2 = BreakRotator(InRot=BreakTransform(InTransform=GetDeltaTransformFromRefPose(self=@n1, BoneName=spine_05, BaseName=spine_01)).Rotation)
   n3 = MapRangeClamped(Value=@n2.Yaw, InRangeA=-30, InRangeB=-150, OutRangeB=250)
   > Set backpack_rotate(backpack_rotate=MakeRotator(Roll=float + float(B=@n3)))
   > Set backpack_translation(backpack_translation=MakeVector(Y=MapRangeClamped(Value=@n2.Yaw, InRangeA=-20, InRangeB=-50, OutRangeB=-9)))
   TryGetPawnOwner()
   FInterpTo(Current=0, Target=0, DeltaTime=0, InterpSpeed=10)
   MapRangeClamped(Value=@n2.Yaw, InRangeA=-8, InRangeB=-30, OutRangeB=10)
   MapRangeClamped(Value=@n2.Yaw, InRangeA=-30, InRangeB=-50, OutRangeB=-5)
  ]}
]
variables [
 backpack_rotate {type=Rotator}
 backpack_translation {type=Vector}
]
